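# Two-sided, lit translucent material for anti-aliased text, sampling the RobotoDistanceField font through a Font parameter.
# The default path centers the distance field at 0.5, then scales edge sharpness and intensity using UV filter width from ComputeFilterWidth, so edges adapt to screen-space size.
# A feature level switch gives ES2 and ES3_1 a cheaper fixed-contrast remap of the font sample, skipping the filter-width math.
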Material /Engine/EngineMaterials/AntiAliasedTextMaterialTranslucent
graph {
 properties {blendMode=BLEND_Translucent, pixelDepthOffsetMode=PDOM_Legacy, twoSided=true, usedShadingModels=Default Lit}
 nodes [
  n1 = Add(A=FontSampleParameter() {font=/Engine/EngineFonts/RobotoDistanceField, parameterName=Font}.R) {constB=-0.5}
  n2 = TextureCoordinate()
  n3 = Constant()
  n4 = Divide(B=MaterialFunctionCall!(In=AppendVector(A=@n2, B=@n3)) {materialFunction=/Engine/Functions/Engine_MaterialFunctions03/Procedurals/ComputeFilterWidth}.Result)
  n5 = LinearInterpolate(Alpha=@n4) {constB=0.05}
  n6 = Multiply(A=Add(A=Multiply(A=@n1, B=@n5) {constB=180}) {constB=0.5}, B=Clamp(None=LinearInterpolate(Alpha=@n4) {constB=0.005})) {constB=180}
  n7 = LinearInterpolate(Alpha=FontSampleParameter() {font=&RobotoDistanceField, parameterName=Font}.R) {constA=-90, constB=90}
  FeatureLevelSwitch(Default=@n6, ES2=@n7, ES3_1=@n7, SM4_REMOVED=@n6)
  VertexColor()
 ]
}
