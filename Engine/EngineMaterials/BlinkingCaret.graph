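# BlinkingCaret is an unlit, translucent engine material that, going by its name, draws a blinking text caret.
# It multiplies the RGB of a SourceTexture parameter, defaulting to a white square, by the absolute value of a time-driven cosine.
# Taking the absolute value folds the wave into a repeating pulse at twice the cosine's frequency, fading smoothly rather than switching hard on and off.
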Material /Engine/EngineMaterials/BlinkingCaret
graph {
 properties {bUsedWithStaticLighting=true, blendMode=BLEND_Translucent, pixelDepthOffsetMode=PDOM_Legacy, shadingModel=MSM_Unlit, usedShadingModels=Unlit}
 nodes [
  n1 = Time()
  Multiply(
   A=TextureSampleParameter2D() {parameterName=SourceTexture, texture=/Engine/EngineResources/WhiteSquareTexture}.RGB
   B=Abs(None=Cosine(None=@n1) {period=2})
  )
 ]
}
