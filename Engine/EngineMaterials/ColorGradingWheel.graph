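# Translucent UI material that draws a color grading wheel as a gradient-colored outer ring over a background texture.
# UV distance from center feeds two opposed SmoothSteps, multiplied into a thin ring mask just inside the unit radius.
# Screen-space derivatives (DDX/DDY) of that distance set the edge softness, antialiasing the ring at any display size.
# The mask blends background into gradient color and is maxed with background alpha for opacity.
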Material /Engine/EngineMaterials/ColorGradingWheel
graph {
 properties {blendMode=BLEND_Translucent, materialDomain=MD_UI, usedShadingModels=Default Lit}
 nodes [
  n1 = TextureSample() {texture=&/ColorGradingWheelBackground}
  n2 = TextureCoordinate()
  n3 = Length(None=Multiply(A=Subtract(A=@n2) {constB=0.5}) {constB=2}) {is=class MaterialExpressionLength}
  n4 = MaterialFunctionCall(X=DDX(Value=@n3), Y=DDY(Value=@n3)) {materialFunction=&MakeFloat2}
  n5 = Subtract(B=Multiply(A=Length(None=@n4.Result)) {constB=1.5})
  n6 = Multiply(
   A=SmoothStep(Max=@n5, Value=@n3) {constMin=1, constValue=0.95}
   B=SmoothStep(Min=@n5, Value=Add(A=@n3) {constB=0.06}) {constMin=1, constValue=0.95}
  )
  LinearInterpolate(A=@n1.RGB, B=TextureSample() {texture=&/ColorGradingWheelGradient}.RGB, Alpha=@n6)
  Max(A=@n1.A, B=@n6)
 ]
}
