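# Unlit engine material built around a cubemap texture parameter, TexCube, which defaults to DefaultCubemap.
# A constant +Z vector passes through a Transform node at its default spaces (tangent to world) to form the cubemap lookup direction.
# The lookup therefore follows the world-space surface normal rather than mesh UVs.
# Usage flags enable it for skeletal meshes, mesh deformers and static lighting.
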Material /Engine/EngineMaterials/CubeMaterial
graph {
 properties {
  bUsedWithMeshDeformer=true
  bUsedWithSkeletalMesh=true
  bUsedWithStaticLighting=true
  pixelDepthOffsetMode=PDOM_Legacy
  shadingModel=MSM_Unlit
  usedShadingModels=Unlit
 }
 nodes [TextureSampleParameterCube!(UVs=Transform(None=Constant3Vector() {constant=<rgba 0 0 1 0>}), Apply View MipBias) {parameterName=TexCube, texture=&/DefaultCubemap}]
}
