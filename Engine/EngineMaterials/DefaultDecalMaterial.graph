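# DefaultDecalMaterial is an engine material, named as the default for decals, with an Unlit shading model.
# It samples the DefaultDiffuse texture and uses two AppendVector nodes to build a green, green, blue vector, replacing red with green.
# The graph shows no connection from that vector to a material output.
# Usage flags cover skeletal meshes, spline meshes, mesh deformers and static lighting.
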Material /Engine/EngineMaterials/DefaultDecalMaterial
graph {
 properties {
  bUsedWithMeshDeformer=true
  bUsedWithSkeletalMesh=true
  bUsedWithSplineMeshes=true
  bUsedWithStaticLighting=true
  opacityMaskClipValue=0.33
  pixelDepthOffsetMode=PDOM_Legacy
  shadingModel=MSM_Unlit
  usedShadingModels=Unlit
 }
 nodes [
  MaterialExpressionTextureSample0 = TextureSample() {texture=&/DefaultDiffuse}
  AppendVector(A=AppendVector(A=@MaterialExpressionTextureSample0.G, B=@MaterialExpressionTextureSample0.G), B=@MaterialExpressionTextureSample0.B)
 ]
}
