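# DefaultLightFunctionMaterial is the engine's default material in the Light Function domain, with an Unlit shading model.
# Its graph holds a single Constant node with a value of one and no textures, parameters or math.
# No link from that constant to a material output appears in the graph.
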Material /Engine/EngineMaterials/DefaultLightFunctionMaterial
graph {
 properties {materialDomain=MD_LightFunction, pixelDepthOffsetMode=PDOM_Legacy, shadingModel=MSM_Unlit, usedShadingModels=Unlit}
 nodes [Constant() {r=1}]
}
