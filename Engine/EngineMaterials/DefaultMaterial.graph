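# Engine DefaultMaterial: a Default Lit grid material enabled for static lighting and many mesh types, including skeletal, instanced, spline and Niagara.
# It samples a packed grid mask at several UV scales to build grayscale tile color, cavity and other scalar terms.
# CameraDepthFade blends between tile scales, changing small-tile and cavity detail with camera distance.
# A grid normal map is softened by scaling down X and Y.
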Material /Engine/EngineMaterials/DefaultMaterial
graph {
 properties {
  bUsedWithInstancedStaticMeshes=true
  bUsedWithMeshDeformer=true
  bUsedWithMeshParticles=true
  bUsedWithNiagaraMeshParticles=true
  bUsedWithNiagaraRibbons=true
  bUsedWithNiagaraSprites=true
  bUsedWithSkeletalMesh=true
  bUsedWithSplineMeshes=true
  bUsedWithStaticLighting=true
  floatPrecisionMode=MFPM_Full_MaterialExpressionOnly
  pixelDepthOffsetMode=PDOM_Legacy
  usedShadingModels=Default Lit
 }
 nodes [
  n1 = Divide(A=TextureCoordinate())
  n2 = TextureSample(UVs=@n1) {samplerType=SAMPLERTYPE_LinearColor, texture=&/T_Default_Material_Grid_M}
  n3 = TextureSample(UVs=Divide(A=@n1) {constB=0.1}) {samplerType=SAMPLERTYPE_LinearColor, texture=&T_Default_Material_Grid_M}
  n4 = MaterialFunctionCall(Fade Length=Constant() {r=1500}) {materialFunction=&CameraDepthFade}
  n5 = LinearInterpolate(A=LinearInterpolate(Alpha=@n3.G) {constA=0.7}, B=Constant() {r=1}, Alpha=@n4.Result)
  Clamp(None=Add(A=LinearInterpolate(Alpha=@n2.G) {constB=0.5}, B=Multiply(A=@n5) {constA=1}))
  LinearInterpolate(A=@n3.A, B=@n2.A, Alpha=@n4.Result) {desc=Uses Cavity mask near camera}
  n6 = LinearInterpolate(Alpha=TextureSample(UVs=TextureCoordinate() {uTiling=20, vTiling=20}) {samplerType=SAMPLERTYPE_LinearColor, texture=&T_Default_Material_Grid_M}.R) {constA=0.4}
  n7 = Constant() {r=0.295}
  n8 = Constant() {r=0.66}
  n9 = LinearInterpolate(A=@n3.G, Alpha=@n4.Result) {constB=0, desc=Add small tiles to basecolor near camera}
  n10 = LinearInterpolate(A=ComponentMask(None=Add(A=@n2.G, B=@n9) {constB=0}) {r=true}, Alpha=Constant() {r=0.5}) {constAlpha=0.875, constB=0.5}
  Multiply(
   A=LinearInterpolate(A=@n6, B=OneMinus(None=@n6), Alpha=LinearInterpolate(A=@n3.G, Alpha=@n4.Result))
   B=Multiply(A=LinearInterpolate(A=@n7, B=@n8, Alpha=@n10), B=Constant() {r=0.5}) {constB=0.5}
  )
  Multiply(
   A=TextureSample(UVs=Divide(A=@n1) {constB=0.05}) {samplerType=SAMPLERTYPE_Normal, texture=&/T_Default_Material_Grid_N}.RGB
   B=Constant3Vector() {constant=<rgba 0.3 0.3 1 0>}
  )
 ]
}
