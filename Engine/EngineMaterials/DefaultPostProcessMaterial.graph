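# Engine default post-process material that overlays the DefaultDiffuse texture as a soft frame along screen edges over the scene image.
# Viewport UV distances from opposite edges are divided by a texel-scaled border size and clamped, and the minimum of their R components forms the blend alpha.
# That alpha lerps from the texture to PostProcessInput0.
# A feature level switch outputs only the texture on ES2 and ES3.1.
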Material /Engine/EngineMaterials/DefaultPostProcessMaterial
graph {
 properties {
  bUsedWithMeshDeformer=true
  bUsedWithSkeletalMesh=true
  bUsedWithStaticLighting=true
  materialDomain=MD_PostProcess
  opacityMaskClipValue=0.33
  pixelDepthOffsetMode=PDOM_Legacy
  shadingModel=MSM_Unlit
  usedShadingModels=Unlit
 }
 nodes [
  n1 = ScreenPosition()
  n2 = Multiply(A=SceneTexelSize(), B=ViewSize())
  MaterialExpressionTextureSample0 = TextureSample(UVs=Divide(A=@n1.ViewportUV, B=@n2)) {texture=&/DefaultDiffuse}
  n3 = Subtract(A=@n2, B=@n1.ViewportUV)
  n4 = Multiply(A=SceneTexelSize(), B=Constant() {desc=Border Size, r=100})
  n5 = MaterialFunctionCall(Float2=Clamp(None=Divide(A=@n3, B=@n4))) {materialFunction=&BreakOutFloat2Components}
  n6 = MaterialFunctionCall(Float2=Clamp(None=Divide(A=@n1.ViewportUV, B=@n4))) {materialFunction=&BreakOutFloat2Components}
  n7 = LinearInterpolate(
   A=@MaterialExpressionTextureSample0.RGB
   B=ComponentMask(None=SceneTexture() {sceneTextureId=PPI_PostProcessInput0}.Color) {b=true, g=true, r=true}
   Alpha=Min(A=Min(A=@n5.R, B=@n5.R), B=Min(A=@n6.R, B=@n6.R))
  ) {constAlpha=0}
  FeatureLevelSwitch(Default=@n7, ES2=@MaterialExpressionTextureSample0.RGB, ES3_1=@MaterialExpressionTextureSample0.RGB, SM4_REMOVED=@n7)
 ]
}
