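# EditorBrushMaterial is an engine material, named for editor brushes, that uses unlit, translucent shading.
# Its graph holds only two constant expressions, a red color vector and a low scalar value, with no links to material outputs shown.
# It has no textures, parameters or math, and its translucency is set to the before depth of field pass.
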Material /Engine/EngineMaterials/EditorBrushMaterial
graph {
 properties {
  blendMode=BLEND_Translucent
  pixelDepthOffsetMode=PDOM_Legacy
  shadingModel=MSM_Unlit
  translucencyPass=MTP_BeforeDOF
  usedShadingModels=Unlit
 }
 nodes [
  Constant3Vector() {constant=<rgba 1 0 0 0>}
  Constant() {r=0.25}
 ]
}
