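# Engine material for emissive meshes, unlit, additive and two-sided, flagged for use with Niagara mesh particles.
# A Color vector parameter, white by default, multiplies the RGB of a texture parameter named LinearColor, which defaults to a white grid texture.
# Its translucency pass is set to render before depth of field.
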Material /Engine/EngineMaterials/EmissiveMeshMaterial
graph {
 properties {
  bUsedWithNiagaraMeshParticles=true
  blendMode=BLEND_Additive
  pixelDepthOffsetMode=PDOM_Legacy
  shadingModel=MSM_Unlit
  translucencyPass=MTP_BeforeDOF
  twoSided=true
  usedShadingModels=Unlit
 }
 nodes [
  MaterialExpressionVectorParameter0 = VectorParameter() {parameterName=Color, defaultValue=<rgba 1 1 1 1>}
  Multiply(
   A=@MaterialExpressionVectorParameter0.RGB
   B=TextureSampleParameter2D() {parameterName=LinearColor, texture=&/DefaultWhiteGrid_Low}.RGB
  )
 ]
}
