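# Engine unlit material EmissiveTexturedMaterial, named for showing a texture as emissive color without scene lighting.
# Its only node is a 2D texture parameter, Texture, defaulting to DefaultDiffuse, so material instances can swap the image.
# The graph lists no math nodes or explicit output links, and the material is flagged for use on mesh particles.
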Material /Engine/EngineMaterials/EmissiveTexturedMaterial
graph {
 properties {bUsedWithMeshParticles=true, pixelDepthOffsetMode=PDOM_Legacy, shadingModel=MSM_Unlit, usedShadingModels=Unlit}
 nodes [TextureSampleParameter2D() {parameterName=Texture, texture=&/DefaultDiffuse}]
}
