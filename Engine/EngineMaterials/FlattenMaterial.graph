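# FlattenMaterial is an engine parent material for flattened surfaces, rebuilding Default Lit channels from textures or constants for static, skeletal, spline and Nanite meshes.
# Static switches choose texture or constant per channel, except the always-sampled normal map.
# Metallic, roughness and specular can share PackedTexture's R, G and B.
# Emissive uses either an HDR texture or scaled color, and nested switches select UV0–6 for all samples.
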Material /Engine/EngineMaterials/FlattenMaterial
graph {
 properties {
  bUseMaterialAttributes=true
  bUsedWithMeshDeformer=true
  bUsedWithMorphTargets=true
  bUsedWithNanite=true
  bUsedWithSkeletalMesh=true
  bUsedWithSplineMeshes=true
  bUsedWithStaticLighting=true
  pixelDepthOffsetMode=PDOM_Legacy
  usedShadingModels=Default Lit
 }
 nodes [
  n1 = TextureCoordinate() {coordinateIndex=2}
  n2 = TextureCoordinate() {coordinateIndex=5}
  n3 = StaticSwitchParameter(True=TextureCoordinate() {coordinateIndex=6}, False=TextureCoordinate()) {parameterName=UseUV6}
  n4 = StaticSwitchParameter(
   True=TextureCoordinate() {coordinateIndex=3}
   False=StaticSwitchParameter(True=TextureCoordinate() {coordinateIndex=4}, False=StaticSwitchParameter(True=@n2, False=@n3) {parameterName=UseUV5}) {parameterName=UseUV4}
  ) {parameterName=UseUV3}
  n5 = StaticSwitchParameter(
   True=StaticSwitchParameter(True=TextureCoordinate() {coordinateIndex=1}, False=StaticSwitchParameter(True=@n1, False=@n4) {parameterName=UseUV2}) {parameterName=UseUV1}
   False=TextureCoordinate()
  ) {parameterName=UseCustomUV}
  n6 = TextureSampleParameter2D(UVs=@n5) {
   parameterName=PackedTexture
   samplerSource=SSM_Clamp_WorldGroupSettings
   samplerType=SAMPLERTYPE_LinearColor
   texture=&/BaseFlattenLinearColor
  }
  n7 = TextureSampleParameter2D(UVs=@n5) {
   parameterName=MetallicTexture
   samplerSource=SSM_Clamp_WorldGroupSettings
   samplerType=SAMPLERTYPE_LinearColor
   texture=&BaseFlattenLinearColor
  }
  n8 = TextureSampleParameter2D(UVs=@n5) {
   parameterName=SpecularTexture
   samplerSource=SSM_Clamp_WorldGroupSettings
   samplerType=SAMPLERTYPE_LinearColor
   texture=&BaseFlattenLinearColor
  }
  n9 = TextureSampleParameter2D(UVs=@n5) {
   parameterName=RoughnessTexture
   samplerSource=SSM_Clamp_WorldGroupSettings
   samplerType=SAMPLERTYPE_LinearColor
   texture=&BaseFlattenLinearColor
  }
  n10 = TextureSampleParameter2D(UVs=@n5) {
   parameterName=EmissiveHDRTexture
   samplerSource=SSM_Clamp_WorldGroupSettings
   samplerType=SAMPLERTYPE_LinearColor
   texture=/Engine/Functions/Engine_MaterialFunctions02/PivotPainter2/Black_1x1_EXR_Texture
  }
  n11 = TextureSampleParameter2D(UVs=@n5) {parameterName=EmissiveColorTexture, samplerSource=SSM_Clamp_WorldGroupSettings, texture=&/BaseFlattenEmissiveMap}
  n12 = StaticSwitchParameter(True=@n11.RGB, False=VectorParameter() {parameterName=EmissiveColorConst}.RGB) {parameterName=UseEmissiveColor}
  n13 = ScalarParameter() {defaultValue=1, parameterName=EmissiveScale}
  n14 = TextureSampleParameter2D(UVs=@n5) {
   parameterName=NormalTexture
   samplerSource=SSM_Clamp_WorldGroupSettings
   samplerType=SAMPLERTYPE_Normal
   texture=&/BaseFlattenNormalMap
  }
  n15 = TextureSampleParameter2D(UVs=@n5) {
   parameterName=OpacityTexture
   samplerType=SAMPLERTYPE_LinearGrayscale
   texture=&/BaseFlattenGrayscaleMap.BaseFlattenGrayScaleMap
  }
  n16 = TextureSampleParameter2D(UVs=@n5) {
   parameterName=OpacityMaskTexture
   samplerType=SAMPLERTYPE_LinearGrayscale
   texture=&BaseFlattenGrayScaleMap
  }
  n17 = TextureSampleParameter2D(UVs=@n5) {
   parameterName=AmbientOcclusionTexture
   samplerType=SAMPLERTYPE_LinearGrayscale
   texture=&BaseFlattenGrayScaleMap
  }
  MakeMaterialAttributes!(
   BaseColor=StaticSwitchParameter(
    True=TextureSampleParameter2D(UVs=@n5) {parameterName=BaseColorTexture, texture=&/BaseFlattenDiffuseMap}.RGB
    False=VectorParameter() {parameterName=BaseColorConst}.RGB
   ) {parameterName=UseBaseColor}
   Metallic=StaticSwitchParameter(True=StaticSwitchParameter(True=@n6.R, False=@n7.RGB) {parameterName=PackMetallic}, False=ScalarParameter() {parameterName=MetallicConst}) {parameterName=UseMetallic}
   Specular=StaticSwitchParameter(
    True=StaticSwitchParameter(True=@n6.B, False=@n8.RGB) {parameterName=PackSpecular}
    False=ScalarParameter() {defaultValue=0.5, parameterName=SpecularConst}
   ) {parameterName=UseSpecular}
   Roughness=StaticSwitchParameter(True=StaticSwitchParameter(True=@n6.G, False=@n9.RGB) {parameterName=PackRoughness}, False=ScalarParameter() {parameterName=RoughnessConst}) {parameterName=UseRoughness}
   Anisotropy
   EmissiveColor=StaticSwitchParameter(True=@n10.RGB, False=Multiply(A=@n12, B=@n13)) {parameterName=UseEmissiveHDR}
   Opacity=StaticSwitchParameter(True=@n15.RGB, False=ScalarParameter() {defaultValue=1, parameterName=OpacityConst}) {parameterName=UseOpacity}
   OpacityMask=StaticSwitchParameter(True=@n16.RGB, False=ScalarParameter() {defaultValue=1, parameterName=OpacityMaskConst}) {parameterName=UseOpacityMask}
   Normal=@n14.RGB
   Tangent
   WorldPositionOffset
   SubsurfaceColor
   ClearCoat
   ClearCoatRoughness
   AmbientOcclusion=StaticSwitchParameter(True=@n17.RGB, False=ScalarParameter() {defaultValue=1, parameterName=AmbientOcclusionConst}) {parameterName=UseAmbientOcclusion}
   Refraction
   CustomizedUVs_0
   CustomizedUVs_1
   CustomizedUVs_2
   CustomizedUVs_3
   CustomizedUVs_4
   CustomizedUVs_5
   CustomizedUVs_6
   CustomizedUVs_7
   PixelDepthOffset
   ShadingModel
   Displacement
  )
 ]
}
