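# Engine material for fog volumes, judging by its name, using unlit shading and additive blending.
# An EmissiveColor vector parameter, pale blue-grey by default, has its RGB multiplied by itself, squaring the color.
# A separate Opacity scalar parameter sits alongside it, with no texture, depth or noise nodes in the graph.
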Material /Engine/EngineMaterials/FogVolumeMaterial
graph {
 properties {blendMode=BLEND_Additive, pixelDepthOffsetMode=PDOM_Legacy, shadingModel=MSM_Unlit, usedShadingModels=Unlit}
 nodes [
  n1 = VectorParameter() {parameterName=EmissiveColor, defaultValue=<rgba 0.5 0.5 0.7 1>}
  Multiply(A=@n1.RGB, B=@n1.RGB)
  ScalarParameter() {defaultValue=1, parameterName=Opacity}
 ]
}
