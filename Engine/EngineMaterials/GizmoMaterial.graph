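# Unlit, two-sided engine material for editor gizmos, flagged for editor compositing, with fake view-based shading and a tintable color.
# A clamped dot product of vertex normal and camera vector drives a half-to-full brightness ramp plus a tight, half-strength power highlight.
# That sum is multiplied by the GizmoColor parameter, white by default.
# A separate constant-offset Add on the dot product feeds no other node.
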Material /Engine/EngineMaterials/GizmoMaterial
graph {
 properties {bUsedWithEditorCompositing=true, pixelDepthOffsetMode=PDOM_Legacy, shadingModel=MSM_Unlit, twoSided=true, usedShadingModels=Unlit}
 nodes [
  n1 = Clamp(None=DotProduct(A=VertexNormalWS(), B=CameraVectorWS()))
  Add(A=@n1) {constB=0.2}
  n2 = Power(Base=@n1) {constExponent=20}
  Multiply(
   A=Add(A=Multiply(A=@n2) {constB=0.5}, B=LinearInterpolate(Alpha=@n1) {constA=0.5})
   B=VectorParameter() {parameterName=GizmoColor, defaultValue=<rgba 1 1 1 1>}.RGB
  )
 ]
}
