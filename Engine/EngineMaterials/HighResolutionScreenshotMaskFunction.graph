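# HighResolutionScreenshotMaskFunction is a material function that turns Custom Depth into a hard binary mask, named for high-resolution screenshot masking.
# It scales the CustomDepth scene texture by the reciprocal of a large constant, clamps it, adds a small bias and floors it.
# Pixels at or near that far threshold and beyond, including empty Custom Depth, become one, while nearer Custom Depth pixels become zero.
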MaterialFunction /Engine/EngineMaterials/HighResolutionScreenshotMaskFunction
graph {nodes [
  n1 = Constant() {r=10000}
  n2 = Clamp(None=Multiply(A=SceneTexture() {sceneTextureId=PPI_CustomDepth}.Color, B=Divide(B=@n1)))
  FunctionOutput(None=Floor(None=Add(A=@n2) {constB=0.01}))
 ]}
