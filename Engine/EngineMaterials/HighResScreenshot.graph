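# Post-process material for high-resolution screenshots that tints the masked area of the scene green.
# A lerp blends the post-process input color toward solid green, using HighResolutionScreenshotMaskFunction as alpha.
# A separate chain scales custom depth by a far-clip distance, then clamps, offsets and floors it into a near-binary mask, but does not feed the lerp.
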Material /Engine/EngineMaterials/HighResScreenshot
graph {
 properties {
  blendMode=BLEND_Translucent
  materialDomain=MD_PostProcess
  pixelDepthOffsetMode=PDOM_Legacy
  shadingModel=MSM_Unlit
  usedShadingModels=Unlit
 }
 nodes [
  n1 = Multiply(A=SceneTexture() {sceneTextureId=PPI_CustomDepth}.Color, B=Divide(B=Constant() {desc=World Space Far Clip Plane Distance, r=10000})) {constB=0.0001}
  Floor(None=Add(A=Clamp(None=@n1)) {constB=0.01})
  LinearInterpolate(
   A=SceneTexture() {sceneTextureId=PPI_PostProcessInput0}.Color
   B=Constant4Vector() {constant=<rgba 0 1 0 1>}.RGBA
   Alpha=MaterialFunctionCall() {materialFunction=&/HighResolutionScreenshotMaskFunction}.Result
  )
 ]
}
