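# Post-process material that highlights the high-resolution screenshot capture region over the scene.
# The RegionRect parameter, scaled by view size and scene texel size, gives rectangle edges that If nodes compare with viewport UV to build an inside mask.
# That mask lerps from desaturated scene color with diagonal Fmod stripes outside to full scene color inside.
# Thin texel-based tests along the rectangle edges darken the inside color to black as a border.
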Material /Engine/EngineMaterials/HighResScreenshotCaptureRegion
graph {
 properties {materialDomain=MD_PostProcess, pixelDepthOffsetMode=PDOM_Legacy, shadingModel=MSM_Unlit, usedShadingModels=Unlit}
 nodes [
  n1 = SceneTexture() {sceneTextureId=PPI_PostProcessInput0}
  n2 = MaterialFunctionCall(Float2=ScreenPosition().ViewportUV) {desc=Break out screen position X and Y, materialFunction=&BreakOutFloat2Components}
  n3 = MaterialFunctionCall(Float2=@n1.Size) {materialFunction=&BreakOutFloat2Components}
  n4 = Divide(A=@n3.G, B=@n3.G)
  n5 = Multiply(A=Add(A=@n2.R, B=Multiply(A=@n2.G, B=@n4))) {constB=50}
  n6 = Constant() {r=2}
  n7 = Multiply(A=Floor(None=Fmod(A=@n5, B=@n6))) {constB=0.25}
  n8 = MaterialFunctionCall(Float2=Multiply(A=ViewSize(), B=SceneTexelSize())) {materialFunction=&BreakOutFloat2Components}
  n9 = VectorParameter() {desc=Input capture region, parameterName=RegionRect, defaultValue=<rgba 0.25 0.25 0.75 0.75>}
  n10 = Clamp(None=Multiply(A=@n8.R, B=@n9.R))
  n11 = Subtract(A=@n2.R, B=@n10)
  n12 = MaterialFunctionCall(Float2=Multiply(A=@n1.InvSize)) {materialFunction=&BreakOutFloat2Components}
  n13 = Constant() {r=1}
  n14 = Constant()
  n15 = Clamp(None=Multiply(A=@n8.R, B=@n9.B))
  n16 = Subtract(A=@n2.R, B=@n15)
  n17 = Multiply(
   A=If(A=Abs(None=@n11), B=@n12.R, "A > B"=@n13, "A == B"=@n13, "A < B"=@n14)
   B=If(A=Abs(None=@n16), B=@n12.R, "A > B"=@n13, "A == B"=@n13, "A < B"=@n14)
  )
  n18 = Clamp(None=Multiply(A=@n8.G, B=@n9.G))
  n19 = Subtract(A=@n2.G, B=@n18)
  n20 = Clamp(None=Multiply(A=@n8.G, B=@n9.A))
  n21 = Subtract(A=@n2.G, B=@n20)
  n22 = Multiply(
   A=If(A=Abs(None=@n19), B=@n12.G, "A > B"=@n13, "A == B"=@n13, "A < B"=@n14)
   B=If(A=Abs(None=@n21), B=@n12.G, "A > B"=@n13, "A == B"=@n13, "A < B"=@n14)
  )
  n23 = Constant() {r=1}
  n24 = Constant()
  n25 = If(A=@n2.R, B=@n10, "A > B"=@n23, "A == B"=@n23, "A < B"=@n24)
  n26 = If(A=@n2.R, B=@n15, "A > B"=@n24, "A == B"=@n23, "A < B"=@n23)
  n27 = If(A=@n2.G, B=@n18, "A > B"=@n23, "A == B"=@n23, "A < B"=@n24)
  n28 = If(A=@n2.G, B=@n20, "A > B"=@n24, "A == B"=@n23, "A < B"=@n23)
  LinearInterpolate(
   A=Multiply(A=Desaturation(None=@n1.Color), B=Add(A=@n7) {constB=0.75})
   B=Multiply(A=Multiply(A=@n17, B=@n22), B=@n1.Color)
   Alpha=Multiply(A=Multiply(A=@n25, B=@n26), B=Multiply(A=@n27, B=@n28))
  )
  TwoSidedSign()
 ]
}
