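# Translucent unlit post-process material for high-resolution screenshot masking, going by its name and function.
# It inverts the Result of the shared HighResolutionScreenshotMaskFunction with a OneMinus node.
# A separate Divide node holds a constant labeled as the world-space far clip plane distance, with no dividend shown.
# The graph shows no links from either node to a material output.
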Material /Engine/EngineMaterials/HighResScreenshotMask
graph {
 properties {
  blendMode=BLEND_Translucent
  materialDomain=MD_PostProcess
  pixelDepthOffsetMode=PDOM_Legacy
  shadingModel=MSM_Unlit
  usedShadingModels=Unlit
 }
 nodes [
  Divide(B=Constant() {desc=World Space Far Clip Plane Distance, r=10000})
  OneMinus(None=MaterialFunctionCall() {materialFunction=&/HighResolutionScreenshotMaskFunction}.Result)
 ]
}
