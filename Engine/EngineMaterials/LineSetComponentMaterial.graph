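# Unlit, two-sided material for line set components that expands lines into camera-facing quads through vertex displacement.
# UV U supplies thickness, in pixels or world units via ScreenSpace, and tangent Z gives the offset's line direction, with separate perspective and orthographic math.
# In perspective, vertices behind the camera slide along the line almost to the camera plane instead of thickening.
# UV V adds depth bias toward the camera, and vertex color is exposure-compensated.
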Material /Engine/EngineMaterials/LineSetComponentMaterial
graph {
 properties {
  bCastRayTracedShadows=false
  pixelDepthOffsetMode=PDOM_Legacy
  shadingModel=MSM_Unlit
  twoSided=true
  usedShadingModels=Unlit
  lightmassSettings {
   bCastShadowAsMasked=false
   bOverrideCastShadowAsMasked=false
   bOverrideDiffuseBoost=false
   bOverrideEmissiveBoost=false
   bOverrideExportResolutionScale=false
   diffuseBoost=0
   emissiveBoost=1
   exportResolutionScale=1
  }
 }
 nodes [
  n1 = ComponentMask(None=TextureCoordinate()) {desc=U coordinate holds line thickness in pixels, r=true}
  n2 = Divide(A=@n1, B=ComponentMask(None=ViewSize()) {r=true})
  n3 = DotProduct(
   A=WorldPosition() {desc=View ray, worldPositionShaderOffset=WPT_CameraRelativeNoOffsets}.XYZ
   B=MaterialFunctionCall() {materialFunction=&CameraDirectionVector}.Result
  ) {desc=Get distance to view plane passing through point}
  n4 = Multiply(
   A=ComponentMask(None=ViewProperty() {desc=Gives us ratio of half width to distance to view plane, property=MEVP_TanHalfFieldOfView}.Property) {r=true}
   B=@n3
  ) {desc=Get half width of view plane passing through point}
  n5 = If(
   A=NamedRerouteUsage()
   "A > B"=@n4
   "A == B"=MaterialFunctionCall!(Half Width) {materialFunction=/Engine/Functions/Engine_MaterialFunctions03/Viewport/MF_OrthoWidth}.Result
   "A < B"=@n4
  ) {constB=1}
  n6 = NamedRerouteUsage()
  n7 = MaterialFunctionCall() {materialFunction=&CameraDirectionVector}
  n8 = Transform(None=Constant3Vector() {constant=<rgba 0 0 1 0>}) {desc=Tangent z holds positive or negative line direction}
  n9 = MaterialFunctionCall!(
   Plane Normal=@n7.Result
   Vector=CrossProduct(A=WorldPosition() {desc=View ray, worldPositionShaderOffset=WPT_CameraRelativeNoOffsets}.XYZ, B=@n8)
  ) {materialFunction=/Engine/Functions/Engine_MaterialFunctions02/Math/ProjectVectorOntoPlane}
  n10 = CrossProduct(A=@n7.Result, B=@n8)
  n11 = Multiply(
   A=LinearInterpolate(
    A=Multiply(A=@n1) {constB=0.5}
    B=Multiply(A=@n2, B=@n5) {desc="Get world-space half thickness in a view plane \r\npassing through point"}
    Alpha=ScalarParameter() {defaultValue=1, parameterName=ScreenSpace}
   )
   B=Normalize(VectorInput=If(A=@n6, "A > B"=@n9.Result, "A == B"=@n10, "A < B"=@n9.Result) {constB=1})
  ) {desc=Move to thicken}
  n12 = Subtract(B=NamedRerouteUsage()) {desc="1 if perspective, 0 if ortho"}
  n13 = NamedRerouteUsage()
  n14 = NamedRerouteUsage()
  n15 = MaterialFunctionCall() {materialFunction=&CameraDirectionVector}
  n16 = Multiply(A=Divide(A=Multiply(A=@n13) {constB=-1}, B=DotProduct(A=@n14, B=@n15.Result) {desc=How fast line moves in normal component}) {desc=t value to intersect plane at camera}) {
   constB=0.95
   desc="For extremely long lines, the imprecision in the stored direction will result in inaccurate intersection point that is noticeable close to the camera, so don't actually move  all the way."
  }
  n17 = If(
   A=NamedRerouteUsage()
   "A > B"=@n11
   "A == B"=@n11
   "A < B"=Multiply(A=Multiply(A=@n12, B=@n16) {desc=Don't do anything if ortho}, B=NamedRerouteUsage()) {desc=Displacement to bring to plane at camera}
  ) {desc="if in front of camera, thicken, if behind, don't thicken and clamp to camera plane"}
  n18 = If(A=NamedRerouteUsage(), "A > B"=@n17, "A == B"=@n11, "A < B"=@n17) {constB=1, desc="For perspective mode, need special handling for verts behind camera plane"}
  n19 = Add(A=NamedRerouteUsage(), B=WorldPosition().XYZ)
  n20 = CameraPositionWS()
  n21 = TextureCoordinate()
  Add(
   A=@n18
   B=MaterialFunctionCall!(
    CameraToVertex=Subtract(A=@n19, B=@n20)
    DepthBias=ComponentMask(None=@n21) {desc="The V coordinate holds depth bias, which we use to move the vertex towards the camera", g=true}
   ) {materialFunction=/MeshModelingToolsetExp/Materials/GetDisplacementVectorForDepthBias}.Result
  ) {desc=Move to apply depth bias}
  EyeAdaptationInverse(LightValueInput=VertexColor()) {desc=Undo effect of exposure}
  n22 = Round(None=IsOrthographic()) {desc="We round here because IsOrthographic produces an integer value\r\n which isn't acceptable by If nodes as input under restrictive mode."}
  NamedRerouteDeclaration(None=@n22) {name=IsOrtho, nodeColor=<rgba 0.4766 1 0 1>}
  NamedRerouteDeclaration(None=@n18) {name=DisplacementToCreateQuad_1, nodeColor=<rgba 0.1953 1 0 1>}
  NamedRerouteDeclaration(None=@n8) {name=LineDirection, nodeColor=<rgba 0 1 0.8125 1>}
  NamedRerouteDeclaration(None=@n3) {name=DistanceToViewPlane, nodeColor=<rgba 0 1 0.0625 1>}
 ]
}
