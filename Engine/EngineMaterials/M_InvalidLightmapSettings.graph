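# M_InvalidLightmapSettings is an engine material that, by its name, flags objects with invalid lightmap settings.
# It samples the InvalidLightmapSettings texture using viewport screen-position UVs multiplied by a constant, so the pattern tiles across the screen independent of mesh UVs.
# The material uses Default Lit shading and is enabled for instanced static meshes.
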Material /Engine/EngineMaterials/M_InvalidLightmapSettings
graph {
 properties {bUsedWithInstancedStaticMeshes=true, pixelDepthOffsetMode=PDOM_Legacy, usedShadingModels=Default Lit}
 nodes [TextureSample(UVs=Multiply(A=ScreenPosition().ViewportUV) {constB=4}) {texture=&/InvalidLightmapSettings}]
}
