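# Engine masked material that, by its name, stands in for hidden mesh sections under Nanite.
# It is a two-sided Default Lit material flagged for use with Nanite and skeletal meshes.
# The graph holds a single purple constant color node, with no connection to any material output shown.
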Material /Engine/EngineMaterials/NaniteHiddenSectionMaterial
graph {
 properties {
  bUsedWithNanite=true
  bUsedWithSkeletalMesh=true
  blendMode=BLEND_Masked
  pixelDepthOffsetMode=PDOM_Legacy
  twoSided=true
  usedShadingModels=Default Lit
 }
 nodes [Constant3Vector() {constant=<rgba 0.6514 0.1356 0.8963 1>}]
}
