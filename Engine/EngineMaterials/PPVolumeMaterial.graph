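# PPVolumeMaterial is an unlit translucent engine material holding a color-grading chain, apparently for post-process volumes, judging by its name.
# The chain subtracts Shadows, clamps, divides by HighLights, raises to the MidTones power, then feeds a Desaturation node.
# As shown, the Subtract has no source color, the Desaturation scalar parameter is unwired, and nothing connects to a material output.
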Material /Engine/EngineMaterials/PPVolumeMaterial
graph {
 properties {bUsedWithStaticLighting=true, blendMode=BLEND_Translucent, pixelDepthOffsetMode=PDOM_Legacy, shadingModel=MSM_Unlit, usedShadingModels=Unlit}
 nodes [
  Comment() {text=Midtones}
  Comment() {text=Desaturation}
  Comment() {text=Highlights}
  Comment() {text=Shadows}
  n1 = Clamp(None=Subtract(B=VectorParameter() {parameterName=Shadows}.RGB), Min=Constant3Vector())
  n2 = VectorParameter() {parameterName=HighLights, defaultValue=<rgba 1 1 1 1>}
  Desaturation(None=Power(Base=Divide(A=@n1, B=@n2.RGB), Exp=VectorParameter() {parameterName=MidTones, defaultValue=<rgba 1 1 1 1>}.RGB))
  ScalarParameter() {parameterName=Desaturation}
 ]
}
