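# Translucent, unlit, two-sided engine material for joint limit visuals in the Physics Asset Tool (PhAT), with depth testing disabled so it draws over scene geometry.
# It fakes lighting by dotting a fixed light direction with the pixel normal, remapping with bias-scale and a small lift, then tinting by the Color parameter.
# A constant supplies a low value, likely opacity, and a CameraOffset function takes an extremely large offset amount.
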Material /Engine/EngineMaterials/PhAT_JointLimitMaterial
graph {
 properties {
  bDisableDepthTest=true
  blendMode=BLEND_Translucent
  pixelDepthOffsetMode=PDOM_Legacy
  shadingModel=MSM_Unlit
  translucencyPass=MTP_BeforeDOF
  twoSided=true
  usedShadingModels=Unlit
 }
 nodes [
  MaterialExpressionConstant4 = Constant() {r=0.3}
  MaterialFunctionCall(Offset Amount=Constant() {r=9.999999884841548e+26}) {materialFunction=/Engine/Functions/Engine_MaterialFunctions02/WorldPositionOffset/CameraOffset}
  n1 = Constant3Vector() {desc=Fake light direction, constant=<rgba 1 1.5 -3 0>}
  n2 = DotProduct(A=Normalize(VectorInput=Multiply(A=@n1) {constB=-1}), B=PixelNormalWS())
  Multiply(A=VectorParameter() {parameterName=Color, defaultValue=<rgba 1 1 1 1>}.RGB, B=Add(A=ConstantBiasScale(None=@n2)) {constB=0.1})
 ]
}
