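# Unlit, two-sided engine material for point set components that draws each point as a camera-facing quad with a view-relative screen size.
# Vertices move along a camera-space direction stored in tangent Z, scaled by U-coordinate point size over view width and the perspective or orthographic half-width at that point.
# The V coordinate supplies a depth bias that shifts the displaced vertex toward the camera.
# Vertex color passes through inverse eye adaptation to cancel exposure.
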Material /Engine/EngineMaterials/PointSetComponentMaterial
graph {
 properties {pixelDepthOffsetMode=PDOM_Legacy, shadingModel=MSM_Unlit, twoSided=true, usedShadingModels=Unlit}
 nodes [
  n1 = ComponentMask(None=TextureCoordinate()) {desc=U coordinate holds point size, r=true}
  n2 = ComponentMask(None=ViewSize()) {r=true}
  n3 = Round(None=IsOrthographic()) {desc="We round here because IsOrthographic produces an integer value\r\n which isn't acceptable by If nodes as input under restrictive mode."}
  n4 = MaterialFunctionCall() {materialFunction=&CameraDirectionVector}
  n5 = MaterialFunctionCall() {desc=Gives us ratio of half width to distance to view plane, materialFunction=/Engine/Functions/Engine_MaterialFunctions02/Utility/FOV}
  n6 = Multiply(
   A=DotProduct(A=Normalize(VectorInput=@n4.Result), B=WorldPosition() {worldPositionShaderOffset=WPT_CameraRelativeNoOffsets}.XYZ) {desc=Get distance to view plane passing through point}
   B=@n5."FOV "
  ) {desc=Get half width of view plane passing through point}
  n7 = MaterialFunctionCall!(Half Width) {materialFunction=/Engine/Functions/Engine_MaterialFunctions03/Viewport/MF_OrthoWidth}
  n8 = Transform(None=Constant3Vector() {constant=<rgba 0 0 1 0>}) {desc=Tangent Z holds direction to move each point in camera space, transformType=TRANSFORM_Local}
  n9 = MaterialFunctionCall() {materialFunction=&ObjectScale}
  n10 = Multiply(
   A=Multiply(A=Divide(A=@n1, B=@n2) {desc=Get ratio of point size to screen width}, B=If(A=@n3, "A > B"=@n6, "A == B"=@n7.Result, "A < B"=@n6) {constB=1}) {desc=Get displacement amount in view plane through point}
   B=Transform(None=Multiply(A=@n8, B=@n9.Scale XYZ) {desc="The local space vector we get here is inversely scaled, so undo the scaling."}) {transformSourceType=TRANSFORMSOURCE_Camera}
  ) {constB=2}
  n11 = Add(A=NamedRerouteUsage(), B=WorldPosition().XYZ)
  n12 = CameraPositionWS()
  n13 = TextureCoordinate()
  Add(
   A=@n10
   B=MaterialFunctionCall!(
    CameraToVertex=Subtract(A=@n11, B=@n12)
    DepthBias=ComponentMask(None=@n13) {desc="The V coordinate holds depth bias, which we use to move the vertex towards the camera", g=true}
   ) {materialFunction=/MeshModelingToolsetExp/Materials/GetDisplacementVectorForDepthBias}.Result
  )
  EyeAdaptationInverse(LightValueInput=VertexColor()) {desc=Undo effect of exposure}
  NamedRerouteDeclaration(None=@n10) {name=DisplacementToCreateQuad, nodeColor=<rgba 1 0.5859 0 1>}
 ]
}
