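# RemoveSurfaceMaterial is an engine material whose name suggests a marker for surfaces to be removed, though the removal logic is not part of the asset.
# Its graph has no expressions, so it does no shading computation of its own.
# It uses the Unlit shading model, is flagged for static lighting and keeps the legacy pixel depth offset mode.
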Material /Engine/EngineMaterials/RemoveSurfaceMaterial
graph {expressions {}, properties {bUsedWithStaticLighting=true, pixelDepthOffsetMode=PDOM_Legacy, shadingModel=MSM_Unlit, usedShadingModels=Unlit}}
