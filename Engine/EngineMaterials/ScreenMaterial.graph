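# ScreenMaterial is an unlit engine material that samples a texture in screen space.
# The ScreenPosition viewport UV, masked to its red and green channels, supplies the UVs for the ScreenTex texture parameter, which defaults to DefaultDiffuse.
# Lookup follows viewport position instead of mesh UVs, so the texture aligns to the screen.
# The material is flagged for use with static lighting.
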Material /Engine/EngineMaterials/ScreenMaterial
graph {
 properties {bUsedWithStaticLighting=true, pixelDepthOffsetMode=PDOM_Legacy, shadingModel=MSM_Unlit, usedShadingModels=Unlit}
 nodes [TextureSampleParameter2D(UVs=ComponentMask(None=ScreenPosition().ViewportUV) {g=true, r=true}) {parameterName=ScreenTex, texture=&/DefaultDiffuse}]
}
