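# Engine volume-domain, additive material for Heterogeneous Volumes, rendering Sparse Volume Texture data such as density and temperature.
# It samples the texture parameter by local position normalized to object bounds, masking Attributes A into density and Attributes B into temperature.
# Density gets an optional clamp and scale, alongside a tinted, scaled albedo.
# Optionally clamped temperature drives blackbody color by default, or a flat tint through a static switch.
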Material /Engine/EngineMaterials/SparseVolumeMaterial
graph {
 properties {
  bUsedWithHeterogeneousVolumes=true
  blendMode=BLEND_Additive
  materialDomain=MD_Volume
  pixelDepthOffsetMode=PDOM_Legacy
  usedShadingModels=Default Lit
 }
 nodes [
  Multiply(
   A=VectorParameter() {group=Albedo, parameterName=Albedo Color, defaultValue=<rgba 1 1 1 0>}.RGB
   B=ScalarParameter() {defaultValue=0.18, group=Albedo, parameterName=Albedo Scale}
  )
  n1 = MaterialFunctionCall() {materialFunction=&LocalPosition}
  n2 = ObjectLocalBounds()
  n3 = SparseVolumeTextureSampleParameter!(UVs=Divide(A=Subtract(A=@n1.Local Position, B=@n2.Min), B=@n2.Extents), Tex) {parameterName=SparseVolumeTexture, sparseVolumeTexture=/Engine/EngineResources/EmptySparseVolumeTexture}
  n4 = StaticComponentMaskParameter(None=StaticSwitchParameter(True=@n3.Attributes A, False=@n3.Attributes A) {defaultValue=true, group=Density Mapping, parameterName="Density (Attributes A)"}) {defaultR=true, group=Density Mapping, parameterName=Density Mask}
  n5 = ScalarParameter() {group=Density, parameterName=Density Clamp Min}
  n6 = ScalarParameter() {defaultValue=1, group=Density, parameterName=Density Clamp Max}
  Multiply(
   A=StaticSwitchParameter(True=Clamp(None=@n4, Min=@n5, Max=@n6), False=@n4) {group=Density, parameterName=Clamp SVT Density}
   B=ScalarParameter() {defaultValue=1, group=Density, parameterName=Density Scale}
  )
  n7 = VectorParameter() {group=Blackbody, parameterName=Blackbody Tint, defaultValue=<rgba 1 1 1 0>}
  n8 = ScalarParameter() {defaultValue=1, group=Blackbody, parameterName=Blackbody Scale}
  n9 = StaticComponentMaskParameter(None=StaticSwitchParameter(True=@n3.Attributes B, False=@n3.Attributes B) {defaultValue=true, group=Temperature Mapping, parameterName="Temperature (Attributes B)"}) {defaultR=true, group=Temperature Mapping, parameterName=Temperature Mask}
  n10 = StaticSwitchParameter(
   True=Clamp(
    None=@n9
    Min=ScalarParameter() {group=Temperature, parameterName=Temperature Clamp Min}
    Max=ScalarParameter() {defaultValue=1, group=Temperature, parameterName=Temperature Clamp Max}
   )
   False=@n9
  ) {group=Temperature, parameterName=Clamp SVT Temperature}
  n11 = Multiply(A=ScalarParameter() {defaultValue=4500, group=Blackbody, parameterName=Blackbody Temperature Scale}, B=@n10)
  n12 = VectorParameter() {group=Temperature, parameterName=Temperature Color, defaultValue=<rgba 1 1 1 0>}
  n13 = ScalarParameter() {defaultValue=1, group=Temperature, parameterName=Temperature Scale}
  StaticSwitchParameter(True=Multiply(A=Multiply(A=@n7.RGB, B=@n8), B=BlackBody!(Temp=@n11)), False=Multiply(A=@n10, B=Multiply(A=@n12.RGB, B=@n13))) {defaultValue=true, group=Blackbody, parameterName=Use Blackbody Temperature}
 ]
}
