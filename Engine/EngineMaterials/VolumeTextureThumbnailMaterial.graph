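# Unlit engine material that previews a volume texture, named for use in volume texture thumbnails.
# A custom HLSL node, ScaleOffset, turns world position into 3D texture coordinates using a Radius scalar parameter.
# It swaps and negates the Y and Z axes, offsets height by the radius, and rescales around 0.5.
# The result drives the UVs of the PreviewVolume texture parameter, so the surface shows a slice through the volume.
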Material /Engine/EngineMaterials/VolumeTextureThumbnailMaterial
graph {
 properties {pixelDepthOffsetMode=PDOM_Legacy, shadingModel=MSM_Unlit, usedShadingModels=Unlit}
 nodes [
  MaterialExpressionScalarParameter_0 = ScalarParameter() {defaultValue=128, parameterName=Radius}
  MaterialExpressionWorldPosition_0 = WorldPosition()
  n1 = Custom!(Radius=@MaterialExpressionScalarParameter_0, WorldPos=@MaterialExpressionWorldPosition_0.XYZ) {
   code="return .5 + .5 * float3(WorldPos.x, -WorldPos.z+ Radius, -WorldPos.y ) /Radius;"
   description=ScaleOffset
   inputs [
    {inputName=Radius, input {expression=&self:MaterialExpressionScalarParameter_0}}
    {inputName=WorldPos, input {expression=&self:MaterialExpressionWorldPosition_0}}
   ]
  }
  TextureSampleParameterVolume!(UVs=@n1, Apply View MipBias) {parameterName=PreviewVolume}
 ]
}
