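# DefaultWaterMaterial is an engine default material for water surfaces, using the Single Layer Water shading model.
# It drives only the Single Layer Water output from parameters, with no other surface inputs connected.
# Scattering is the parameter's RGB scaled by its alpha over 100, and absorption is the reciprocal of the squared Absorption RGB, so larger values absorb less.
# An Anisotropy scalar sets the phase function, while color scale behind water is left unconnected.
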Material /Engine/EngineMaterials/WaterMaterial.DefaultWaterMaterial
graph {
 properties {bUsedWithWater=true, pixelDepthOffsetMode=PDOM_Legacy, shadingModel=MSM_SingleLayerWater, usedShadingModels=SingleLayerWater}
 nodes [
  n1 = VectorParameter() {parameterName=Scattering, defaultValue=<rgba 0.1 0.1 0.1 0.5>}
  n2 = VectorParameter() {parameterName=Absorption, defaultValue=<rgba 1 8 20 100>}
  SingleLayerWaterMaterialOutput!(
   ScatteringCoefficients=Multiply(A=@n1.RGB, B=Divide(A=@n1.A) {constB=100})
   AbsorptionCoefficients=Divide(B=Multiply(A=@n2.RGB, B=@n2.RGB))
   PhaseG=ScalarParameter() {defaultValue=0.5, parameterName=Anisotropy}
   ColorScaleBehindWater
  )
 ]
}
